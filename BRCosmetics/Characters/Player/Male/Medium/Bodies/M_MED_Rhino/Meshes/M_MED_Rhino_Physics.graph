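# Physics asset for the M_MED_Rhino male player body cosmetic, set up for the dangling back flap.
# A box body on dyn_Flap_Back is constrained to a tiny capsule on Flap_Back_Driven_Bone with tight, asymmetric swing and twist limits.
# That driven bone joins a large pelvis capsule through a constraint with free swing and twist, so only the flap joint restricts rotation.
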PhysicsAsset /BRCosmetics/Characters/Player/Male/Medium/Bodies/M_MED_Rhino/Meshes/M_MED_Rhino_Physics
bodies {
 Flap_Back_Driven_Bone [{shapeType=Capsule, length=0.505, radius=0.505, rotation=<rot 0 -90 0>}]
 dyn_Flap_Back [{shapeType=Box, extentX=4.6295, extentY=17.5504, extentZ=30.7412, center=<-0.0136 -1.0188 -10.6961>, rotation=<rot 0 90.1082 -89.4134>}]
 pelvis [{shapeType=Capsule, length=19.2245, radius=34.8917, center=<-5.1046 22.1499 0>, rotation=<rot 0 -102.3575 0>}]
}
constraints [
 {bone1Name=dyn_Flap_Back, bone2Name=Flap_Back_Driven_Bone, swing1LimitDegrees=35, swing2LimitDegrees=15, twistLimitDegrees=5}
 {
  bone1Name=Flap_Back_Driven_Bone
  bone2Name=pelvis
  swing1LimitDegrees=45
  swing1Motion=Free
  swing2LimitDegrees=45
  swing2Motion=Free
  twistLimitDegrees=45
  twistMotion=Free
 }
]
