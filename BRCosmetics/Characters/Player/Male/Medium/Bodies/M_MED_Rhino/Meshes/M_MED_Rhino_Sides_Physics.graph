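# Physics asset for the side meshes of the male medium Rhino body cosmetic, covering its hanging flaps and the lower body.
# Capsules shape the pelvis, both thighs and only the right calf, while tiny capsules sit at the left, right and back flap roots.
# Box bodies on the two dyn_flappy bones hang from their side roots with limited swing and asymmetric twist.
# All other joints leave swing and twist free.
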PhysicsAsset /BRCosmetics/Characters/Player/Male/Medium/Bodies/M_MED_Rhino/Meshes/M_MED_Rhino_Sides_Physics
bodies {
 Flap_Back_Driven_Bone [{shapeType=Capsule, length=0.505, radius=0.505, rotation=<rot 0 -90 0>}]
 calf_r [{shapeType=Capsule, length=23.3994, radius=9.8743, center=<19.4364 4.2778 -0.8924>, rotation=<rot 0.2853 -83.5253 -86.2177>}]
 dyn_Flap_Left [{shapeType=Capsule, length=0.505, radius=0.505, rotation=<rot 0 -90 0>}]
 dyn_Flap_Right [{shapeType=Capsule, length=0.505, radius=0.505, rotation=<rot 0 -90 0>}]
 dyn_flappy_l [{shapeType=Box, extentX=20.4631, extentY=4.6322, extentZ=39.7895, center=<17.0118 0.01 0>, rotation=<rot 0 -87.9938 -89.7912>}]
 dyn_flappy_r [{shapeType=Box, extentX=12.2131, extentY=4.6322, extentZ=34.7894, center=<-17 0 0>, rotation=<rot 0 -87.9938 -89.7912>}]
 pelvis [{shapeType=Capsule, length=8.9968, radius=11.6401, center=<0.8371 -0.3363 0.045>, rotation=<rot -2.1646 -90 -89.18>}]
 thigh_l [{shapeType=Capsule, length=15.1387, radius=9.4694, center=<-18.8262 -0.6121 0.0195>, rotation=<rot 87.7675 -180 -180>}]
 thigh_r [{shapeType=Capsule, length=27.9942, radius=8.523, center=<17.4792 0.3227 0.2598>, rotation=<rot 87.2225 180 180>}]
}
constraints [
 {
  bone1Name=Flap_Back_Driven_Bone
  bone2Name=pelvis
  swing1LimitDegrees=45
  swing1Motion=Free
  swing2LimitDegrees=45
  swing2Motion=Free
  twistLimitDegrees=45
  twistMotion=Free
 }
 {bone1Name=dyn_flappy_l, bone2Name=dyn_Flap_Left, swing1LimitDegrees=70, swing2LimitDegrees=70, twistLimitDegrees=5}
 {bone1Name=dyn_flappy_r, bone2Name=dyn_Flap_Right, swing1LimitDegrees=70, swing2LimitDegrees=70, twistLimitDegrees=30}
 {
  bone1Name=dyn_Flap_Right
  bone2Name=pelvis
  swing1LimitDegrees=45
  swing1Motion=Free
  swing2LimitDegrees=45
  swing2Motion=Free
  twistLimitDegrees=45
  twistMotion=Free
 }
 {
  bone1Name=dyn_Flap_Left
  bone2Name=pelvis
  swing1LimitDegrees=45
  swing1Motion=Free
  swing2LimitDegrees=45
  swing2Motion=Free
  twistLimitDegrees=45
  twistMotion=Free
 }
 {
  bone1Name=thigh_l
  bone2Name=pelvis
  swing1LimitDegrees=45
  swing1Motion=Free
  swing2LimitDegrees=45
  swing2Motion=Free
  twistLimitDegrees=45
  twistMotion=Free
 }
 {
  bone1Name=thigh_r
  bone2Name=pelvis
  swing1LimitDegrees=45
  swing1Motion=Free
  swing2LimitDegrees=45
  swing2Motion=Free
  twistLimitDegrees=45
  twistMotion=Free
 }
 {
  bone1Name=calf_r
  bone2Name=thigh_r
  swing1LimitDegrees=45
  swing1Motion=Free
  swing2LimitDegrees=45
  swing2Motion=Free
  twistLimitDegrees=45
  twistMotion=Free
 }
]
