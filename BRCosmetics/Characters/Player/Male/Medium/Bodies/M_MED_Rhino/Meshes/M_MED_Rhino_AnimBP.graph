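# Animation Blueprint for the Rhino medium male cosmetic body, adding secondary motion to a pose copied from the attached parent mesh.
# RBF pose drivers move the shoulder pads using skydive or pistol pose assets, chosen by bIsSkyDiving.
# Weighted constraints position exhausts, thigh-driven nodes swing side flaps, and two RigidBody simulations follow before caching.
# The front flap then blends between standing, crouched and crouch-moving corrections; the EventGraph is empty.
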AnimBlueprint /BRCosmetics/Characters/Player/Male/Medium/Bodies/M_MED_Rhino/Meshes/M_MED_Rhino_AnimBP
target_skeleton=/Game/Characters/Player/Male/Medium/Bodies/M_MED_Rhino/Meshes/M_MED_Rhino_Skeleton
graphs [
 AnimGraph {nodes [
   n1 = "Transform (Modify) Bone - Bone: Flap_Front_Driven_Bone"(
    ComponentPose=Local To Component(LocalPose=Use cached pose 'CrouchCache'())
    Alpha=Get CrouchSpeedAlpha() -> (CrouchSpeedAlpha:float)
    Translation="0,0,0"
    Rotation="32.680,26.356,-52.200"
    Scale="1,1,1"
   ) {
    rotationMode=BMM_Additive
    rotationSpace=BCS_BoneSpace
    translationMode=BMM_Additive
    translationSpace=BCS_BoneSpace
    alphaBoolBlend {blendOption=Linear}
    alphaScaleBiasClamp {bInterpResult=true}
    boneToModify {boneName=Flap_Front_Driven_Bone}
    rotation=<rot 32.68 26.356 -52.2>
   }
   n2 = Component To Local(ComponentPose=@n1)
   n3 = "Transform (Modify) Bone - Bone: Flap_Front_Driven_Bone"(ComponentPose=Local To Component(LocalPose=Use cached pose 'CrouchCache'()), Alpha=1, Translation="0,0,7", Rotation="0,0,-40", Scale="1,1,1") {
    rotationMode=BMM_Additive
    rotationSpace=BCS_BoneSpace
    translationMode=BMM_Additive
    translationSpace=BCS_BoneSpace
    alphaBoolBlend {blendOption=Linear}
    boneToModify {boneName=Flap_Front_Driven_Bone}
    rotation {roll=-40}
    translation {z=7}
   }
   n4 = Flap Node - Bone: dyn_Flap_Front!(ComponentPose:Component Space Pose Link Structure=@n3, Alpha:float=1) -> (Pose:Component Space Pose Link Structure) {
    is=title FortAnimGraphNode_Flap
    bUseClamp=true
    targetClampMax=100
    alphaBoolBlend {blendOption=Linear}
    sourceDrives [
     {
      bUseRange=false
      drivingCurve=null
      multiplier=0.5
      rangeMax=1
      rangeMin=-1
      remappedMax=1
      remappedMin=0
      sourceComponent=RotationZ
      useQuaternion=true
      sourceBone {boneName=thigh_l}
     }
     {
      bUseRange=false
      drivingCurve=null
      multiplier=0.5
      rangeMax=1
      rangeMin=-1
      remappedMax=1
      remappedMin=0
      sourceComponent=RotationZ
      useQuaternion=true
      sourceBone {boneName=thigh_r}
     }
    ]
    targetBone {boneName=dyn_Flap_Front}
   }
   n5 = "dyn_Flap_Front.rotateZ = thigh_l.rotateZ * -0.5 - Bone Driven Controller"!(ComponentPose:Component Space Pose Link Structure=@n4, Alpha:float=1) -> (Pose:Component Space Pose Link Structure) {
    is=title AnimGraphNode_BoneDrivenController
    bAffectTargetRotationZ=true
    multiplier=-0.5
    sourceComponent=RotationZ
    alphaBoolBlend {blendOption=Linear}
    sourceBone {boneName=thigh_l}
    targetBone {boneName=dyn_Flap_Front}
   }
   n6 = "dyn_Flap_Front.rotateZ = thigh_r.rotateZ * 0.5 - Bone Driven Controller"!(ComponentPose:Component Space Pose Link Structure=@n5, Alpha:float=1) -> (Pose:Component Space Pose Link Structure) {
    is=title AnimGraphNode_BoneDrivenController
    bAffectTargetRotationZ=true
    multiplier=0.5
    sourceComponent=RotationZ
    alphaBoolBlend {blendOption=Linear}
    sourceBone {boneName=thigh_r}
    targetBone {boneName=dyn_Flap_Front}
   }
   n7 = Component To Local(ComponentPose=@n6)
   n8 = Get bIsCrouchMoving() -> (bIsCrouchMoving:bool)
   n9 = "dyn_Flap_Front.rotateZ = curve(thigh_l.rotateZ) - Bone Driven Controller"!(ComponentPose:Component Space Pose Link Structure=Local To Component(LocalPose=Use cached pose 'CrouchCache'()), Alpha:float=0.75) -> (Pose:Component Space Pose Link Structure) {
    is=title AnimGraphNode_BoneDrivenController
    bAffectTargetRotationZ=true
    drivingCurve=/Game/Characters/Player/Male/XLarge/Bodies/M_XLG_Jim_01/Meshes/SK_M_XLG_Jim_Flap_L_Thigh_Front
    sourceComponent=RotationZ
    alphaBoolBlend {blendOption=Linear}
    sourceBone {boneName=thigh_l}
    targetBone {boneName=dyn_Flap_Front}
   }
   n10 = "dyn_Flap_Front.rotateZ = curve(thigh_r.rotateZ) - Bone Driven Controller"!(ComponentPose:Component Space Pose Link Structure=@n9, Alpha:float=0.75) -> (Pose:Component Space Pose Link Structure) {
    is=title AnimGraphNode_BoneDrivenController
    bAffectTargetRotationZ=true
    drivingCurve=/Game/Characters/Player/Male/XLarge/Bodies/M_XLG_Jim_01/Meshes/SK_M_XLG_Jim_Flap_R_Thigh_Front
    sourceComponent=RotationZ
    alphaBoolBlend {blendOption=Linear}
    sourceBone {boneName=thigh_r}
    targetBone {boneName=dyn_Flap_Front}
   }
   n11 = Flap Node - Bone: Flap_Front_Driven_Bone!(ComponentPose:Component Space Pose Link Structure=@n10, Alpha:float=0.95) -> (Pose:Component Space Pose Link Structure) {
    is=title FortAnimGraphNode_Flap
    bUseClamp=true
    targetClampMax=3.14
    alphaBoolBlend {blendOption=Linear}
    sourceDrives [
     {
      bUseRange=false
      drivingCurve=null
      multiplier=1
      rangeMax=1
      rangeMin=-1
      remappedMax=1
      remappedMin=0
      sourceComponent=RotationZ
      useQuaternion=true
      sourceBone {boneName=thigh_l}
     }
     {
      bUseRange=false
      drivingCurve=null
      multiplier=1
      rangeMax=1
      rangeMin=-1
      remappedMax=1
      remappedMin=0
      sourceComponent=RotationZ
      useQuaternion=true
      sourceBone {boneName=thigh_r}
     }
    ]
    targetBone {boneName=Flap_Front_Driven_Bone}
   }
   Output Pose(Result=Blend Poses by bool(
     BlendPose_0=Blend Poses by bool(BlendPose_0=@n7, BlendPose_1=@n2, bActiveValue=@n8) {blendType=Linear}
     BlendPose_1=Component To Local(ComponentPose=@n11)
     bActiveValue=Get bIsCrouching()
    ) {blendType=Linear}) {layerGroup=None}
   Save cached pose 'AnimCache'(Pose=Copy Pose From Mesh() {bUseAttachedParent=true}) {cacheName=AnimCache}
   n12 = Pose Driver - Source: clavicle_r!(SourcePose:Pose Link Structure=Use cached pose 'AnimCache'()) -> (Pose:Pose Link Structure) {
    is=title AnimGraphNode_PoseDriver
    poseAsset=&/M_MED_Rhino_Pistol_PoseAsset
    onlyDriveBones [{boneName=dyn_shoulderpad_r}]
    poseTargets [
     {
      bApplyCustomCurve=false
      bIsHidden=false
      distanceMethod=DefaultMethod
      drivenName=bind
      functionType=DefaultFunction
      targetScale=1
      boneTransforms [
       {targetRotation=<rot -85.1914 41.4375 143.0618>, targetTranslation=<4.3535 -0.1466 2.3222>}
       {targetRotation=<rot -64.7854 -3.089 6.0959>, targetTranslation=<-15.5099 0 0.0017>}
      ]
      customCurve {defaultValue=3.4028234663852886e+38, postInfinityExtrap=RCCE_Constant, preInfinityExtrap=RCCE_Constant, keys []}
      targetRotation=<rot 0 0 0>
     }
     {
      bApplyCustomCurve=false
      bIsHidden=false
      distanceMethod=DefaultMethod
      drivenName=pistol_fwd
      functionType=DefaultFunction
      targetScale=0.6761
      boneTransforms [
       {targetRotation=<rot -69.1504 -125.8682 -55.5842>, targetTranslation=<4.3535 -0.1466 2.3222>}
       {targetRotation=<rot -10.7593 101.6029 -55.8541>, targetTranslation=<-15.5099 0 0.0017>}
      ]
      customCurve {defaultValue=3.4028234663852886e+38, postInfinityExtrap=RCCE_Constant, preInfinityExtrap=RCCE_Constant, keys []}
      targetRotation=<rot 0 0 0>
     }
     {
      bApplyCustomCurve=false
      bIsHidden=false
      distanceMethod=DefaultMethod
      drivenName=pistol_up
      functionType=DefaultFunction
      targetScale=0.6761
      boneTransforms [
       {targetRotation=<rot -68.9943 -127.6254 -53.9156>, targetTranslation=<4.3535 -0.1466 2.3222>}
       {targetRotation=<rot 29.2484 104.845 -75.5914>, targetTranslation=<-15.5099 0 0.0017>}
      ]
      customCurve {defaultValue=3.4028234663852886e+38, postInfinityExtrap=RCCE_Constant, preInfinityExtrap=RCCE_Constant, keys []}
      targetRotation=<rot 0 0 0>
     }
     {
      bApplyCustomCurve=false
      bIsHidden=false
      distanceMethod=DefaultMethod
      drivenName=pisto_dwn
      functionType=DefaultFunction
      targetScale=0.688
      boneTransforms [
       {targetRotation=<rot -68.3051 -111.2351 -69.7765>, targetTranslation=<4.3535 -0.1466 2.3222>}
       {targetRotation=<rot -55.846 103.304 -53.1482>, targetTranslation=<-15.5099 0 0.0017>}
      ]
      customCurve {defaultValue=3.4028234663852886e+38, postInfinityExtrap=RCCE_Constant, preInfinityExtrap=RCCE_Constant, keys []}
      targetRotation=<rot 0 0 0>
     }
    ]
    rBFParams {distanceMethod=Quaternion, radius=40}
    sourceBones [{boneName=clavicle_r}, {boneName=upperarm_r}]
   }
   n13 = Pose Driver - Source: clavicle_l!(SourcePose:Pose Link Structure=@n12) -> (Pose:Pose Link Structure) {
    is=title AnimGraphNode_PoseDriver
    poseAsset=&M_MED_Rhino_Pistol_PoseAsset
    onlyDriveBones [{boneName=dyn_shoulderpad_l}]
    poseTargets [
     {
      bApplyCustomCurve=false
      bIsHidden=false
      distanceMethod=DefaultMethod
      drivenName=bind
      functionType=DefaultFunction
      targetScale=0.5281
      boneTransforms [
       {targetRotation=<rot -85.1913 -138.5624 143.0618>, targetTranslation=<4.3575 -0.1466 -2.3223>}
       {targetRotation=<rot -64.7854 -3.089 6.0959>, targetTranslation=<15.5092 -0.0002 -0.0013>}
      ]
      customCurve {defaultValue=3.4028234663852886e+38, postInfinityExtrap=RCCE_Constant, preInfinityExtrap=RCCE_Constant, keys []}
      targetRotation=<rot 0 0 0>
     }
     {
      bApplyCustomCurve=false
      bIsHidden=false
      distanceMethod=DefaultMethod
      drivenName=pistol_fwd
      functionType=DefaultFunction
      targetScale=0.6982
      boneTransforms [
       {targetRotation=<rot -84.2011 51.9203 -48.9848>, targetTranslation=<4.3575 -0.1466 -2.3223>}
       {targetRotation=<rot -40.3083 69.1488 -68.9383>, targetTranslation=<15.5092 -0.0002 -0.0013>}
      ]
      customCurve {defaultValue=3.4028234663852886e+38, postInfinityExtrap=RCCE_Constant, preInfinityExtrap=RCCE_Constant, keys []}
      targetRotation=<rot 0 0 0>
     }
     {
      bApplyCustomCurve=false
      bIsHidden=false
      distanceMethod=DefaultMethod
      drivenName=pistol_up
      functionType=DefaultFunction
      targetScale=1
      boneTransforms [
       {targetRotation=<rot -52.705 -3.3338 10.9169>, targetTranslation=<4.3575 -0.1466 -2.3223>}
       {targetRotation=<rot 0.6145 58.2102 -97.5496>, targetTranslation=<15.5092 -0.0002 -0.0013>}
      ]
      customCurve {defaultValue=3.4028234663852886e+38, postInfinityExtrap=RCCE_Constant, preInfinityExtrap=RCCE_Constant, keys []}
      targetRotation=<rot 0 0 0>
     }
     {
      bApplyCustomCurve=false
      bIsHidden=false
      distanceMethod=DefaultMethod
      drivenName=pisto_dwn
      functionType=DefaultFunction
      targetScale=0.5281
      boneTransforms [
       {targetRotation=<rot -84.4046 115.6771 -113.8102>, targetTranslation=<4.3575 -0.1466 -2.3223>}
       {targetRotation=<rot -79.5266 108.0831 -92.2137>, targetTranslation=<15.5092 -0.0002 -0.0013>}
      ]
      customCurve {defaultValue=3.4028234663852886e+38, postInfinityExtrap=RCCE_Constant, preInfinityExtrap=RCCE_Constant, keys []}
      targetRotation=<rot 0 0 0>
     }
    ]
    rBFParams {distanceMethod=Quaternion, radius=40}
    sourceBones [{boneName=clavicle_l}, {boneName=upperarm_l}]
   }
   n14 = Pose Driver - Source: clavicle_r(SourcePose=Use cached pose 'AnimCache'()) {
    poseAsset=&/M_MED_Rhino_Skydive_PoseAsset
    onlyDriveBones [{boneName=dyn_shoulderpad_r}]
    poseTargets [
     {
      bApplyCustomCurve=false
      bIsHidden=false
      distanceMethod=DefaultMethod
      drivenName=bind
      functionType=DefaultFunction
      targetScale=1
      boneTransforms [
       {targetRotation=<rot -85.1914 41.4375 143.0618>, targetTranslation=<4.3535 -0.1466 2.3222>}
       {targetRotation=<rot -64.7854 -3.089 6.0959>, targetTranslation=<-15.5099 0 0.0017>}
      ]
      customCurve {defaultValue=3.4028234663852886e+38, postInfinityExtrap=RCCE_Constant, preInfinityExtrap=RCCE_Constant, keys []}
      targetRotation=<rot 0 0 0>
     }
     {
      bApplyCustomCurve=false
      bIsHidden=false
      distanceMethod=DefaultMethod
      drivenName=falling
      functionType=DefaultFunction
      targetScale=0.5753
      boneTransforms [
       {targetRotation=<rot -64.5761 87.1291 99.0962>, targetTranslation=<4.3535 -0.1466 2.3222>}
       {targetRotation=<rot -1.7208 1.6478 -58.0537>, targetTranslation=<-15.5099 0 0.0017>}
      ]
      customCurve {defaultValue=3.4028234663852886e+38, postInfinityExtrap=RCCE_Constant, preInfinityExtrap=RCCE_Constant, keys []}
      targetRotation=<rot 0 0 0>
     }
     {
      bApplyCustomCurve=false
      bIsHidden=false
      distanceMethod=DefaultMethod
      drivenName=floating
      functionType=DefaultFunction
      targetScale=0.5753
      boneTransforms [
       {targetRotation=<rot -66.3529 -164.2714 -10.8528>, targetTranslation=<4.3535 -0.1466 2.3222>}
       {targetRotation=<rot 25.1298 0.919 -74.8956>, targetTranslation=<-15.5099 0 0.0017>}
      ]
      customCurve {defaultValue=3.4028234663852886e+38, postInfinityExtrap=RCCE_Constant, preInfinityExtrap=RCCE_Constant, keys []}
      targetRotation=<rot 0 0 0>
     }
    ]
    rBFParams {distanceMethod=Quaternion, radius=55}
    sourceBones [{boneName=clavicle_r}, {boneName=upperarm_r}]
   }
   n15 = Pose Driver - Source: clavicle_l(SourcePose=@n14) {
    poseAsset=&M_MED_Rhino_Skydive_PoseAsset
    onlyDriveBones [{boneName=dyn_shoulderpad_l}]
    poseTargets [
     {
      bApplyCustomCurve=false
      bIsHidden=false
      distanceMethod=DefaultMethod
      drivenName=bind
      functionType=DefaultFunction
      targetScale=1
      boneTransforms [
       {targetRotation=<rot -85.1913 -138.5624 143.0618>, targetTranslation=<4.3575 -0.1466 -2.3223>}
       {targetRotation=<rot -64.7854 -3.089 6.0959>, targetTranslation=<15.5092 -0.0002 -0.0013>}
      ]
      customCurve {defaultValue=3.4028234663852886e+38, postInfinityExtrap=RCCE_Constant, preInfinityExtrap=RCCE_Constant, keys []}
      targetRotation=<rot 0 0 0>
     }
     {
      bApplyCustomCurve=false
      bIsHidden=false
      distanceMethod=DefaultMethod
      drivenName=falling
      functionType=DefaultFunction
      targetScale=0.5328
      boneTransforms [
       {targetRotation=<rot -64.8583 -92.9708 99.1837>, targetTranslation=<4.3575 -0.1466 -2.3223>}
       {targetRotation=<rot -7.818 0.5486 -69.9961>, targetTranslation=<15.5092 -0.0002 -0.0013>}
      ]
      customCurve {defaultValue=3.4028234663852886e+38, postInfinityExtrap=RCCE_Constant, preInfinityExtrap=RCCE_Constant, keys []}
      targetRotation=<rot 0 0 0>
     }
     {
      bApplyCustomCurve=false
      bIsHidden=false
      distanceMethod=DefaultMethod
      drivenName=floating
      functionType=DefaultFunction
      targetScale=0.5328
      boneTransforms [
       {targetRotation=<rot -64.3116 15.5722 -10.6253>, targetTranslation=<4.3575 -0.1466 -2.3223>}
       {targetRotation=<rot 19.5509 -0.8236 -73.2014>, targetTranslation=<15.5092 -0.0002 -0.0013>}
      ]
      customCurve {defaultValue=3.4028234663852886e+38, postInfinityExtrap=RCCE_Constant, preInfinityExtrap=RCCE_Constant, keys []}
      targetRotation=<rot 0 0 0>
     }
    ]
    rBFParams {distanceMethod=Quaternion, radius=55}
    sourceBones [{boneName=clavicle_l}, {boneName=upperarm_l}]
   }
   n16 = Get bIsSkyDiving()
   n17 = "Constraint - exhaust_l (PPPPPPPP)"!(
    ComponentPose=Local To Component(LocalPose=Blend Poses by bool(BlendPose_0=@n15, BlendPose_1=@n13, bActiveValue=@n16) {blendType=Linear})
    Alpha=1
    ConstraintWeights_0=0.0349
    ConstraintWeights_1:float=0.1273
    ConstraintWeights_2:float=0.4764
    ConstraintWeights_3:float=0.035
    ConstraintWeights_4:float=0.2587
    ConstraintWeights_5:float=0.0517
    ConstraintWeights_6:float=0.0129
    ConstraintWeights_7:float=0.0031
   ) -> (Pose) {
    alphaBoolBlend {blendOption=Linear}
    boneToModify {boneName=exhaust_l}
    constraintSetup [
     {targetBone {boneName=upperarm_r}}
     {targetBone {boneName=clavicle_l}}
     {targetBone {boneName=spine_03}}
     {targetBone {boneName=spine_04}}
     {targetBone {boneName=spine_05}}
     {targetBone {boneName=upperarm_twist_01_l}}
     {targetBone {boneName=clavicle_r}}
     {targetBone {boneName=neck_01}}
    ]
    constraintWeights [1,1,1,1,1,1,1,1]
   }
   n18 = "Constraint - exhaust_r (PPPPPP)"!(
    ComponentPose=@n17
    Alpha=1
    ConstraintWeights_0=0.0141
    ConstraintWeights_1:float=0.5216
    ConstraintWeights_2:float=0.0384
    ConstraintWeights_3:float=0.2833
    ConstraintWeights_4:float=0.1393
    ConstraintWeights_5:float=0.0034
   ) -> (Pose) {
    alphaBoolBlend {blendOption=Linear}
    boneToModify {boneName=exhaust_r}
    constraintSetup [
     {targetBone {boneName=clavicle_l}}
     {targetBone {boneName=spine_03}}
     {targetBone {boneName=spine_04}}
     {targetBone {boneName=spine_05}}
     {targetBone {boneName=clavicle_r}}
     {targetBone {boneName=neck_01}}
    ]
    constraintWeights [1,1,1,1,1,1]
   }
   n19 = Flap Node - Bone: Flap_Left_Driven_Bone!(ComponentPose:Component Space Pose Link Structure=@n18, Alpha:float=1) -> (Pose:Component Space Pose Link Structure) {
    is=title FortAnimGraphNode_Flap
    targetClampMax=3.14
    targetComponent=RotationY
    alphaBoolBlend {blendOption=Linear}
    sourceDrives [{
      bUseRange=false
      drivingCurve=null
      multiplier=-0.75
      rangeMax=1
      rangeMin=-1
      remappedMax=1
      remappedMin=0
      sourceComponent=RotationY
      useQuaternion=true
      sourceBone {boneName=thigh_l}
     }]
    targetBone {boneName=Flap_Left_Driven_Bone}
   }
   n20 = Flap Node - Bone: Flap_Right_Driven_Bone!(ComponentPose:Component Space Pose Link Structure=@n19, Alpha:float=1) -> (Pose:Component Space Pose Link Structure) {
    is=title FortAnimGraphNode_Flap
    targetComponent=RotationY
    alphaBoolBlend {blendOption=Linear}
    sourceDrives [{
      bUseRange=false
      drivingCurve=null
      multiplier=-0.5
      rangeMax=1
      rangeMin=-1
      remappedMax=1
      remappedMin=0
      sourceComponent=RotationY
      useQuaternion=false
      sourceBone {boneName=thigh_r}
     }]
    targetBone {boneName=Flap_Right_Driven_Bone}
   }
   n21 = "Transform (Modify) Bone - Bone: Flap_Left_Parent_Space"(
    ComponentPose=@n20
    Alpha=Get ThighPadAlpha() -> (ThighPadAlpha:float)
    Translation="4.097,0.264,0.817"
    Rotation="-14.981,-13.142,1.092"
    Scale="1,1,1"
   ) {
    rotationMode=BMM_Additive
    rotationSpace=BCS_BoneSpace
    translationMode=BMM_Additive
    alphaBoolBlend {blendOption=Linear}
    alphaScaleBiasClamp {bInterpResult=true, interpSpeedDecreasing=30, interpSpeedIncreasing=30}
    boneToModify {boneName=Flap_Left_Parent_Space}
    rotation=<rot -14.981 -13.1422 1.092>
    translation=<4.097 0.264 0.817>
   }
   n22 = RigidBody(ComponentPose=@n21) {
    bOverrideWorldGravity=true
    overridePhysicsAsset=&/M_MED_Rhino_Physics
    simulationSpace=BaseBoneSpace
    alphaBoolBlend {blendOption=Linear}
    baseBoneRef {boneName=spine_01}
    componentLinearAccScale=<1 1 1>
    componentLinearVelScale=<1 1 1>
    overrideWorldGravity {z=-980}
   }
   n23 = RigidBody(ComponentPose=@n22, Alpha=0.9) {
    bOverrideWorldGravity=true
    overridePhysicsAsset=&/M_MED_Rhino_Sides_Physics
    simulationSpace=BaseBoneSpace
    alphaBoolBlend {blendOption=Linear}
    baseBoneRef {boneName=root}
    componentAppliedLinearAccClamp=<100 100 100>
    componentLinearAccScale=<1 1 1>
    componentLinearVelScale=<1 1 1>
    overrideWorldGravity {z=-980}
   }
   Save cached pose 'CrouchCache'(Pose=Component To Local(ComponentPose=@n23)) {cacheName=CrouchCache}
  ]}
 EventGraph {links [], nodes []}
]
variables []
