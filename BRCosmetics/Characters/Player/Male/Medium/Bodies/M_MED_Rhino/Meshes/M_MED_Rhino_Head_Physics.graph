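# Physics asset for the head mesh of the M_MED_Rhino player cosmetic body, covering its antenna.
# It has two capsule bodies, one on the driver_antenna bone and a larger one on dyn_antenna.
# A single constraint joins them with tight swing and twist limits, so the antenna can bend only slightly relative to its driver bone.
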PhysicsAsset /BRCosmetics/Characters/Player/Male/Medium/Bodies/M_MED_Rhino/Meshes/M_MED_Rhino_Head_Physics
bodies {
 driver_antenna [{shapeType=Capsule, length=0.505, radius=0.505, rotation=<rot 0 -90 0>}]
 dyn_antenna [{shapeType=Capsule, length=1.3737, radius=0.7185, center=<0.1029 0 0>, rotation=<rot 0 90 -90>}]
}
constraints [{bone1Name=dyn_antenna, bone2Name=driver_antenna, swing1LimitDegrees=10, swing2LimitDegrees=10, twistLimitDegrees=5}]
