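# Animation Blueprint for the Rhino male cosmetic head mesh, adding physics secondary motion on top of the body's animation.
# It copies the pose from the attached parent mesh, then constrains driver_antenna to a weighted mix of clavicle, upper spine and neck bones.
# A partially blended RigidBody node then simulates the head's physics asset in root-bone space with overridden gravity.
# The event graph holds no logic.
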AnimBlueprint /BRCosmetics/Characters/Player/Male/Medium/Bodies/M_MED_Rhino/Meshes/M_MED_Rhino_Head_AnimBP
target_skeleton=&/M_MED_Rhino_Head_Skeleton
graphs [
 AnimGraph {nodes [
   n1 = "Constraint - driver_antenna (PPPPPPP)"!(
    ComponentPose=Local To Component(LocalPose=Copy Pose From Mesh() {bUseAttachedParent=true})
    Alpha=1
    ConstraintWeights_0=0.2492
    ConstraintWeights_1:float=0.0697
    ConstraintWeights_2:float=0.0175
    ConstraintWeights_3:float=0.3202
    ConstraintWeights_4:float=0.2531
    ConstraintWeights_5:float=0.0282
    ConstraintWeights_6:float=0.062
   ) -> (Pose) {
    alphaBoolBlend {blendOption=Linear}
    boneToModify {boneName=driver_antenna}
    constraintSetup [
     {targetBone {boneName=clavicle_l}}
     {targetBone {boneName=spine_03}}
     {targetBone {boneName=spine_04}}
     {targetBone {boneName=spine_05}}
     {targetBone {boneName=clavicle_r}}
     {targetBone {boneName=neck_02}}
     {targetBone {boneName=neck_01}}
    ]
    constraintWeights [1,1,1,1,1,1,1]
   }
   n2 = RigidBody(ComponentPose=@n1, Alpha=0.6) {
    bOverrideWorldGravity=true
    overridePhysicsAsset=&/M_MED_Rhino_Head_Physics
    simulationSpace=BaseBoneSpace
    alphaBoolBlend {blendOption=Linear}
    baseBoneRef {boneName=root}
    componentAppliedLinearAccClamp=<1000 1000 1000>
    componentLinearAccScale=<1 1 1>
    componentLinearVelScale=<1 1 1>
    overrideWorldGravity {z=-980}
   }
   Output Pose(Result=Component To Local(ComponentPose=@n2)) {layerGroup=None}
  ]}
 EventGraph {nodes [
   TryGetPawnOwner()
   Event BlueprintUpdateAnimation()
  ]}
]
variables []
